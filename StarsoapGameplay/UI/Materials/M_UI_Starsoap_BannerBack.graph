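# Translucent UI material for a Starsoap banner backdrop, drawing a two-colour gradient with stripes and a soft vignette.
# Cosine stripes in offset polar coordinates are clipped by distance-field masks and added to a radial glow to blend Color1 and Color2.
# An edge mask, aspect-corrected from UV channel 3, darkens that colour.
# A sharpened copy of the mask, scaled by Opacity, is inverted and added to the result.
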Material /StarsoapGameplay/UI/Materials/M_UI_Starsoap_BannerBack
graph {
 properties {
  bUsedWithStaticMesh=false
  blendMode=BLEND_Translucent
  materialDomain=MD_UI
  numCustomizedUVs=3
  pixelDepthOffsetMode=PDOM_Legacy
  usedShadingModels=Default Lit
 }
 nodes [
  n1 = VectorParameter() {parameterName=Color2, defaultValue=<rgba 0.0012 0.01 0.0281 1>}
  n2 = VectorParameter() {parameterName=Color1, defaultValue=<rgba 0.0462 0.3531 0.0643 1>}
  n3 = Constant2Vector() {g=0.1}
  n4 = TextureCoordinate()
  n5 = Constant() {r=0.5}
  n6 = Distance(A=Multiply(A=Constant2Vector() {g=1, r=0.9}, B=Add(A=@n3, B=@n4)), B=@n5)
  n7 = Saturate(None=Multiply(A=Add(A=@n6) {constB=-0.5, desc=size}) {constB=-2, desc=sharpness})
  n8 = Add(A=@n4, B=Constant2Vector() {g=-0.6})
  n9 = Constant2Vector() {g=0.25, r=1}
  n10 = Add(A=Distance(A=@n5, B=Multiply(A=@n8, B=@n9))) {constB=-0.54, desc=size}
  n11 = Multiply(A=Multiply(A=@n7, B=@n7) {constB=-5, desc=sharpness}, B=Saturate(None=Multiply(A=@n10) {constB=20, desc=sharpness}))
  n12 = MaterialFunctionCall!(UV Coord=Add(A=@n4, B=Constant2Vector() {g=-2}), LenghtScale, RadialScale) {materialFunction=/Game/Global/MaterialFunctions/TechArt/MF_PolarCoordinates}
  n13 = Multiply(
   A=Add(A=ComponentMask(None=@n12.Result) {g=true}) {constB=0.5, desc=center}
   B=ScalarParameter() {defaultValue=100, parameterName=StripeFrequency}
  ) {constB=100, desc=line frequency}
  n14 = Saturate(None=Add(A=Cosine(None=@n13), B=ScalarParameter() {defaultValue=0.2, parameterName=StripeThickness}) {constB=0.2, desc=line thickness})
  n15 = Distance(A=TextureCoordinate(), B=Constant() {r=0.5})
  n16 = Multiply(A=Add(A=Multiply(A=@n15, B=@n15) {constB=-2}) {constB=-0.2, desc=size}) {constB=-5, desc=sharpness}
  n17 = Multiply(A=Multiply(A=@n16, B=@n16) {constB=-3}, B=@n16) {constB=-3}
  n18 = ScalarParameter() {defaultValue=1, parameterName=GradientBrightness}
  n19 = Add(
   A=Multiply(A=Multiply(A=@n11, B=@n14), B=ScalarParameter() {defaultValue=1, parameterName=StripesBrightness}) {constB=0.7, desc=brightness}
   B=Max(A=Multiply(A=@n17, B=@n18) {constB=0.7, desc=brightness}) {constB=0}
  )
  n20 = Constant() {r=1}
  n21 = TextureCoordinate() {coordinateIndex=3}
  n22 = Divide(A=ComponentMask(None=@n21) {g=true}, B=ComponentMask(None=@n21) {r=true})
  n23 = Multiply(A=Add(A=TextureCoordinate()) {constB=-0.5}, B=Multiply(A=Constant2Vector() {g=1.5, r=1}, B=AppendVector(A=@n20, B=@n22)))
  n24 = Multiply(A=Distance(A=Add(A=@n23) {constB=0.5}, B=Constant() {r=0.5}), B=Add(A=ScalarParameter() {parameterName=BorderOffset}) {constB=2}) {constB=2, desc=soft}
  n25 = Multiply(A=@n24, B=@n24) {constB=-5, desc=soft}
  n26 = OneMinus(None=Multiply(A=@n25, B=@n25) {constB=-5, desc=soft})
  n27 = Saturate(None=@n26)
  n28 = ScalarParameter() {defaultValue=1, parameterName=DarkVignette}
  n29 = ScalarParameter() {defaultValue=1.5, parameterName=OpacitySharpness}
  n30 = Saturate(None=Multiply(A=@n26, B=Divide(A=@n29, B=@n22)) {constB=4, desc=sharpness})
  n31 = ScalarParameter() {defaultValue=1, parameterName=Opacity}
  Add(
   A=Multiply(A=LinearInterpolate(A=@n1.RGB, B=@n2.RGB, Alpha=@n19), B=LinearInterpolate(B=@n27, Alpha=@n28) {constA=1})
   B=OneMinus(None=Multiply(A=@n30, B=@n31))
  )
 ]
}
