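# Translucent UI material for a Starsoap character portrait in a circular frame, with a crosshair icon and a transition-driven cracking effect.
# The portrait texture composites over a radial-gradient background, with a stroked frame from radial-gradient thresholds and a side-clipped upper overflow mask.
# CracksTransition grows cracks and displaces their UVs using a packed crack texture, resampled and compared with the static sample to mask gaps between pieces.
# The crosshair thresholds two offset SDF samples for its stroked shape.
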Material /StarsoapGameplay/UI/Materials/M_UI_Starsoap_Portrait
graph {
 properties {
  bUsedWithStaticMesh=false
  blendMode=BLEND_Translucent
  materialDomain=MD_UI
  numCustomizedUVs=3
  pixelDepthOffsetMode=PDOM_Legacy
  usedShadingModels=Default Lit
 }
 nodes [
  Constant() {r=1}
  n1 = TextureCoordinate()
  n2 = NamedRerouteUsage()
  n3 = TextureSample(UVs=@n1) {desc=Static cracks, samplerType=SAMPLERTYPE_LinearColor, texture=/StarsoapGameplay/UI/Proto/Textures/T_UI_Starsoap_Cracks}
  n4 = ComponentMask(None=@n3.RGB) {g=true, r=true}
  n5 = NamedRerouteDeclaration(None=Multiply(A=Multiply(A=@n2) {constB=-0.2, desc=movement multiplier}, B=Add(A=@n4) {constB=-0.5}) {constB=-0.1}) {name=CracksUVDifference, nodeColor=<rgba 1 0.1172 0 1>}
  n6 = TextureSample(UVs=NamedRerouteDeclaration(None=Add(A=@n1, B=@n5)) {name=CrackedUV, nodeColor=<rgba 0 0.0391 1 1>}) {desc=Moving cracks, samplerType=SAMPLERTYPE_LinearColor, texture=&T_UI_Starsoap_Cracks}
  n7 = NamedRerouteUsage()
  n8 = OneMinus(None=Ceil(None=@n7)) {desc="1 if CrackValue = 0, 0 for the rest"}
  n9 = Max(
   A=Add(
    A=Multiply(A=@n6.B, B=@n3.A) {constA=1, desc=Apply static borders to dynamic cracks SDF}
    B=Multiply(A=@n7) {constB=-0.4, desc=cracks growth}
   )
   B=@n8
  ) {desc=no cracks when 0}
  n10 = Max(A=@n8, B=Add(A=Multiply(A=@n7) {constB=-0.3, desc=border growth}, B=@n3.A) {constB=-0.1}) {desc=no cracks when 0}
  n11 = ComponentMask(None=@n6.RGB) {g=true, r=true}
  n12 = Add(A=Abs(None=Subtract(A=@n11, B=@n4))) {constB=-0.05, desc=similarity tolerance}
  n13 = Add(A=ComponentMask(None=@n12) {r=true}, B=ComponentMask(None=@n12) {g=true})
  n14 = Multiply(A=Saturate(None=Multiply(A=@n10) {constA=1, constB=8, desc=border sharpness}), B=OneMinus(None=Ceil(None=@n13))) {constA=1, desc=cover pixels between borders}
  n15 = Multiply(A=Saturate(None=Multiply(A=@n9) {constB=4, desc=cracks sharpness}), B=@n14) {constA=1, desc=black empty parts}
  n16 = NamedRerouteUsage()
  n17 = NamedRerouteUsage()
  n18 = NamedRerouteUsage()
  LinearInterpolate(B=LinearInterpolate(A=Multiply(A=@n15, B=@n16), B=NamedRerouteUsage(), Alpha=@n17), Alpha=Max(A=Multiply(A=@n18, B=@n14), B=@n17)) {constA=1}
  MaterialFunctionCall() {materialFunction=&DebugTimeSine}
  MaterialFunctionCall(Texture Scale=ScalarParameter() {defaultValue=1, parameterName=GlobalScale}) {materialFunction=&ScaleUVsByCenter}
  n19 = Constant() {desc=shapness, r=5}
  n20 = Multiply(A=Add(A=TextureCoordinate(), B=Constant2Vector() {desc=offset, g=-0.55, r=-0.1})) {constB=3, desc=size}
  n21 = Constant() {desc=sdf distance, r=0.15}
  n22 = Subtract(
   A=TextureSample(UVs=@n20) {samplerSource=SSM_Clamp_WorldGroupSettings, texture=/Game/Athena/Items/Accolades/Materials/Textures/Spree_Alt128px}.R
   B=@n21
  ) {constB=0.2}
  n23 = Min(A=@n21, B=Constant() {desc=stoke, r=0.2})
  n24 = Constant2Vector() {desc=offset, g=-0.05, r=-0.05}
  n25 = Subtract(
   A=TextureSample(UVs=Add(A=@n20, B=@n24)) {samplerSource=SSM_Clamp_WorldGroupSettings, texture=&Spree_Alt128px}.R
   B=@n21
  ) {constB=0.2}
  n26 = Max(A=Multiply(A=@n19, B=Add(A=@n22, B=@n23)) {constB=10}, B=Multiply(A=@n19, B=Add(A=@n23, B=@n25)) {constB=10})
  NamedRerouteDeclaration(None=Multiply(A=Saturate(None=@n26), B=ScalarParameter() {defaultValue=1, group=Crosshair, parameterName=CrosshairIconOpacity})) {name=CrosshairAlpha, nodeColor=<rgba 1 0.3047 0 1>}
  n27 = VectorParameter() {parameterName=BackgroundRGB, defaultValue=<rgba 0.013 0.3613 0.6376 1>}
  n28 = VectorParameter() {parameterName=HighlightRGB, defaultValue=<rgba 0.4452 1 1 1>}
  n29 = Multiply(A=MaterialFunctionCall(UVs=NamedRerouteUsage(), CenterPosition=Constant2Vector() {g=0.75, r=0.5}, Radius=Constant() {r=0.75}, Density=Constant() {r=1.25}) {materialFunction=&RadialGradientExponential}.RadialGradientExponential) {constB=0.5}
  n30 = TextureCoordinate()
  n31 = ScalarParameter() {defaultValue=1, parameterName=PortraitScale}
  n32 = Constant()
  n33 = ScalarParameter() {defaultValue=-0.0982, parameterName=PortraitOffset}
  n34 = Add(
   A=NamedRerouteUsage()
   B=Add(
    A=MaterialFunctionCall(UVs=@n30, Texture Scale=@n31) {materialFunction=&ScaleUVsByCenter}.UVs
    B=AppendVector(A=@n32, B=@n33)
   ) {constB=0}
  )
  n35 = TextureSampleParameter2D(UVs=@n34) {
   parameterName=SmallPreview
   samplerSource=SSM_Clamp_WorldGroupSettings
   texture=/Game/UI/Foundation/Textures/Icons/Heroes/Athena/Soldier/T-Soldier-HID-005-Athena-Commando-M
  }
  n36 = MaterialFunctionCall!(UVs=@n34, Alpha=@n35.A) {materialFunction=/Game/UI/Foundation/Items/ClampAlphaToTransparentBorder}
  NamedRerouteDeclaration(None=LinearInterpolate(
    A=LinearInterpolate(A=@n27.RGB, B=@n28.RGB, Alpha=@n29)
    B=@n35.RGB
    Alpha=NamedRerouteDeclaration(None=@n36.Result) {name=Use Large Preview Image, nodeColor=<rgba 1 0 0.2812 1>}
   )) {name=PortraitRGB, nodeColor=<rgba 1 0.9609 0 1>}
  NamedRerouteDeclaration(None=Saturate(None=ScalarParameter() {parameterName=CracksTransition})) {name=CracksValue, nodeColor=<rgba 1 0 0.6094 1>}
  n37 = Multiply(A=@n22, B=@n19) {constB=10}
  NamedRerouteDeclaration(None=LinearInterpolate(
    A=VectorParameter() {group=Stroke, parameterName=StrokeColor, defaultValue=<rgba 0 0 0 1>}.RGB
    B=VectorParameter() {group=Crosshair, parameterName=CrosshairColor, defaultValue=<rgba 1 1 1 1>}.RGB
    Alpha=Saturate(None=@n37)
   )) {name=CrosshairRGB, nodeColor=<rgba 0 0.0156 1 1>}
  n38 = VectorParameter() {group=Stroke, parameterName=StrokeColor, defaultValue=<rgba 0 0 0 1>}
  n39 = ScalarParameter() {defaultValue=32, parameterName=FrameSharpness}
  n40 = MaterialFunctionCall(UVs=NamedRerouteUsage()) {materialFunction=&RadialGradientExponential}
  n41 = ScalarParameter() {defaultValue=0.54, parameterName=FrameBackgroundSize}
  n42 = Add(A=Subtract(A=@n40.RadialGradientExponential, B=@n41), B=ScalarParameter() {defaultValue=0.04, parameterName=FrameInnerStroke})
  n43 = Add(A=@n42, B=ScalarParameter() {defaultValue=0.07, parameterName=FrameThickness})
  n44 = LinearInterpolate(
   A=@n38.RGB
   B=VectorParameter() {group=Stroke, parameterName=FrameColor, defaultValue=<rgba 0.624 0.2664 0.0015 1>}.RGB
   Alpha=Saturate(None=Multiply(A=@n39, B=@n43))
  )
  n45 = Saturate(None=Multiply(A=@n39, B=@n42))
  n46 = Subtract(A=@n40.RadialGradientExponential, B=@n41)
  n47 = ScalarParameter() {defaultValue=32, parameterName=FrameSharpness}
  n48 = Max(A=NamedRerouteUsage(), B=Saturate(None=Multiply(A=@n46, B=@n47)))
  NamedRerouteDeclaration(None=LinearInterpolate(A=LinearInterpolate(A=@n44, B=@n38.RGB, Alpha=@n45), B=NamedRerouteUsage(), Alpha=@n48)) {name=PortraitFrameRGB, nodeColor=<rgba 0 1 1 1>}
  n49 = ScalarParameter() {defaultValue=0.07, parameterName=FrameOuterStroke}
  n50 = Saturate(None=Multiply(A=@n47, B=Add(A=@n43, B=@n49)))
  NamedRerouteDeclaration(None=Max(A=@n48, B=NamedRerouteDeclaration(None=@n50) {name=FrameAlpha, nodeColor=<rgba 1 0.6797 0 1>})) {name=PortraitFrameAlpha, nodeColor=<rgba 0.3594 0 1 1>}
  n51 = NamedRerouteUsage()
  n52 = NamedRerouteUsage()
  n53 = OneMinus(None=Round(None=ComponentMask(None=@n52) {g=true}))
  n54 = AppendVector(A=ScalarParameter() {defaultValue=0.62, group=PortaitMask, parameterName=Mask Side Clip}, B=Constant() {r=1})
  NamedRerouteDeclaration(None=Multiply(
    A=Multiply(A=@n51, B=@n53)
    B=MaterialFunctionCall(UVs=@n52, Texture Scale=@n54) {materialFunction=&ScaleUVsByCenter}.0-1 mask
   )) {name=PortraitOverflowAlpha, nodeColor=<rgba 0.6406 0 1 1>}
  Reroute(None=@n3.B)
  Reroute(None=@n3.R)
 ]
}
