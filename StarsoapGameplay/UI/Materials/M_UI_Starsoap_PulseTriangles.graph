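# Translucent UI material that draws a warning icon with animated, repeating triangle pulses around it.
# A triangle SDF function plus scrolled time feeds Frac to form moving bordered, semi-filled bands, which fade out with distance.
# The icon texture has scale and offset parameters, and its alpha blends pulse color with icon color and is maxed with the pulse mask for opacity.
# PulseActive gates the pulse, leaving only the icon when zero.
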Material /StarsoapGameplay/UI/Materials/M_UI_Starsoap_PulseTriangles
graph {
 properties {
  bUsedWithStaticMesh=false
  blendMode=BLEND_Translucent
  materialDomain=MD_UI
  numCustomizedUVs=3
  pixelDepthOffsetMode=PDOM_Legacy
  usedShadingModels=Default Lit
 }
 nodes [
  n1 = TextureCoordinate()
  n2 = Add(A=Multiply(A=Add(A=@n1) {constB=-0.5}, B=ScalarParameter() {defaultValue=1.2, group=Icon, parameterName=IconScale})) {constB=0.5}
  n3 = AppendVector(
   A=ScalarParameter() {defaultValue=-0.025, group=Icon, parameterName=IconPosX}
   B=ScalarParameter() {defaultValue=-0.03, group=Icon, parameterName=IconPosY}
  )
  n4 = TextureSample(UVs=Saturate(None=Add(A=@n2, B=@n3))) {texture=/StarsoapGameplay/UI/Proto/Textures/T_UI_Starsoap_Warning}
  n5 = ScalarParameter() {group=Pulse, parameterName=PulseActive}
  n6 = Add(A=@n4.A, B=Floor(None=OneMinus(None=@n5)))
  LinearInterpolate(
   A=VectorParameter() {group=Pulse, parameterName=PulseColor, defaultValue=<rgba 1 0.0648 0.0545 1>}.RGB
   B=@n4.RGB
   Alpha=Saturate(None=Multiply(A=@n6) {constB=2})
  )
  n7 = MaterialFunctionCall!(
   UVs=Add(A=@n1, B=Constant2Vector() {g=-0.016})
   Radius=ScalarParameter() {defaultValue=0.5, group=Pulse, parameterName=CornerRadius}
   Glow Min
   Glow Max
   Corner roundness
   Stroke thickness
   Is stroke thickness relative to size
  ) {materialFunction=/StarsoapGameplay/UI/MaterialFunctions/MF_UI_SDF_Triangle}
  n8 = Multiply(A=ScalarParameter() {defaultValue=2.7, group=Pulse, parameterName=FadeSharpness}, B=@n7.SDF) {constB=2.5}
  n9 = Saturate(None=Multiply(A=OneMinus(None=@n8)) {constB=3})
  n10 = Multiply(A=Time(), B=ScalarParameter() {defaultValue=-0.2, group=Pulse, parameterName=PulseSpeed})
  n11 = Frac(None=Multiply(A=Add(A=@n7.SDF, B=@n10), B=ScalarParameter() {defaultValue=6, group=Pulse, parameterName=PulseRepetition}))
  n12 = Multiply(A=OneMinus(None=@n11)) {constB=3, desc=border thickness}
  n13 = OneMinus(None=@n12)
  n14 = Add(A=Multiply(A=Min(A=@n11, B=@n12)) {constB=0.5, desc=fillopacity}, B=Saturate(None=Min(A=@n12, B=@n13)))
  Max(A=@n4.A, B=Multiply(A=@n5, B=Multiply(A=@n9, B=@n14)))
 ]
}
